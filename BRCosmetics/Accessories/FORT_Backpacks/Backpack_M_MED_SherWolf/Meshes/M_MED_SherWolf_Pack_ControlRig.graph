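# Control Rig for the SherWolf backpack cosmetic that shifts the pack bone in response to the wearer's clavicle poses.
# Two spherical pose readers measure the left and right dyn_clavicle bones against root, using mirrored driver axes and rotation offsets with elongated active and falloff regions.
# Each reader's output weights a Modify Transforms step that rotates and nudges the pack, corrective-style, carrying its child zipper bones along.
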ControlRigBlueprint /BRCosmetics/Accessories/FORT_Backpacks/Backpack_M_MED_SherWolf/Meshes/M_MED_SherWolf_Pack_ControlRig
graphs [
 RigVMFunctionLibrary {links [], nodes []}
 RigVMModel {nodes [
   RigUnit_BeginExecution = &RigUnit_BeginExecution()
   n1 = &RigUnit_SphericalPoseReader(
    ActiveRegionScaleFactors=(PositiveWidth=1.5,NegativeWidth=1.5,PositiveHeight=0.4,NegativeHeight=0.4)
    ActiveRegionSize=0.2
    Debug=(bDrawDebug=true,bDraw2D=False,bDrawLocalAxes=False,DebugScale=25,DebugSegments=20,DebugThickness=0.25)
    DriverAxis=(X=1,Y=0,Z=0)
    DriverItem=(Type=Bone,Name="dyn_clavicle_l")
    FalloffRegionScaleFactors=(PositiveWidth=2,NegativeWidth=2,PositiveHeight=0.35,NegativeHeight=0.5)
    FalloffSize=0.3
    FlipHeightScaling=False
    FlipWidthScaling=False
    OptionalParentItem=(Type=Bone,Name="root")
    RotationOffset=(X=0,Y=-90,Z=-35)
    ExecutePin=@RigUnit_BeginExecution.ExecutePin
   )
   > &RigUnit_ModifyTransforms(
    ItemToModify=((Item=(Type=Bone,Name="pack"),Transform=(Rotation=(X=0.0515,Y=-0.0091,Z=-0.1734,W=0.9835),Translation=(X=0,Y=-0.5,Z=0),Scale3D=(X=1,Y=1,Z=1))))
    WeightMinimum=0
    Weight=@n1.OutputParam
   )
   > n2 = &RigUnit_SphericalPoseReader(
    ActiveRegionScaleFactors=(PositiveWidth=1.5,NegativeWidth=1.5,PositiveHeight=0.4,NegativeHeight=0.4)
    ActiveRegionSize=0.2
    Debug=(bDrawDebug=true,bDraw2D=False,bDrawLocalAxes=False,DebugScale=25,DebugSegments=20,DebugThickness=0.25)
    DriverAxis=(X=-1,Y=0,Z=0)
    DriverItem=(Type=Bone,Name="dyn_clavicle_r")
    FalloffRegionScaleFactors=(PositiveWidth=2,NegativeWidth=2,PositiveHeight=0.35,NegativeHeight=0.5)
    FalloffSize=0.3
    FlipHeightScaling=false
    FlipWidthScaling=false
    OptionalParentItem=(Type=Bone,Name="root")
    RotationOffset=(X=0,Y=90,Z=35)
   )
   > &RigUnit_ModifyTransforms(
    ItemToModify=((Item=(Type=Bone,Name="pack"),Transform=(Rotation=(X=0.0172,Y=0.0027,Z=0.1564,W=0.9875),Translation=(X=0,Y=-0.5,Z=0),Scale3D=(X=1,Y=1,Z=1))))
    WeightMinimum=0
    Weight=@n2.OutputParam
   )
  ]}
]
hierarchy [
 root {loc=<0 0 0>, rot=<0 0 0 1>}
 pack {parent=root, loc=<0 -0.7316 0>, rot=<0 0 0 1>}
 base_zipper_a {parent=pack, loc=<-10.1076 -2.6208 -2.6184>, rot=<0.2706 -0.6533 -0.2706 -0.6533>}
 dyn_zipper_a_01 {parent=base_zipper_a, loc=<0 0 0>, rot=<0 0 0 1>}
 dyn_zipper_a_02 {parent=dyn_zipper_a_01, loc=<0.8059 0 -0.3227>, rot=<0 0 0 1>}
 base_zipper_b {parent=pack, loc=<10.0893 -2.6025 -2.605>, rot=<0.2706 -0.6533 -0.2706 -0.6533>}
 dyn_zipper_b_01 {parent=base_zipper_b, loc=<0 0 0>, rot=<0 0 0 1>}
 dyn_zipper_b_02 {parent=dyn_zipper_b_01, loc=<0.8193 -0.3485 0>, rot=<0 0 0 1>}
 dyn_clavicle_l {parent=root, loc=<2 -0.7316 7.1925>, rot=<0 0 0 1>}
 dyn_clavicle_r {parent=root, loc=<-2 -0.7316 7.1925>, rot=<0 0 0 1>}
]
