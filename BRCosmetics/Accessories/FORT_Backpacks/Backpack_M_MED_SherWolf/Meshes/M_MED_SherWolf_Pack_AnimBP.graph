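# Anim Blueprint for the SherWolf backpack cosmetic, adding zipper physics and fitting it to the wearer's body.
# Each update computes a gravity override and caches the pawn body mesh's world-space clavicle transforms.
# The AnimGraph chains four Anim Dynamics zipper bones with spherical and planar limits, then pins the pack's clavicle bones to those transforms.
# An additive pack offset blends in during skydive floating, and a Control Rig finishes the pose.
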AnimBlueprint /BRCosmetics/Accessories/FORT_Backpacks/Backpack_M_MED_SherWolf/Meshes/M_MED_SherWolf_Pack_AnimBP
target_skeleton=&/M_MED_SherWolf_Pack_Skeleton
graphs [
 AnimGraph {nodes [
   n1 = Get Gravity Override()
   n2 = Anim Dynamics - Bone: dyn_zipper_a_01!(ComponentPose=Local To Component(LocalPose=Input Pose()), Alpha=1, GravityOverride:Vector=@n1, ExternalForce="0,0,0") -> (Pose) {
    angularDampingOverride=0.9
    bOverrideAngularDamping=true
    bUseGravityOverride=true
    bUsePlanarLimit=false
    simulationSpace=RootRelative
    alphaBoolBlend {blendOption=Linear}
    boundBone {boneName=dyn_zipper_a_01}
    componentAppliedLinearAccClamp=<600 600 600>
    componentLinearAccScale=<0.4 0.4 0.4>
    componentLinearVelScale=<0.5 0.5 0.5>
    gravityOverride {z=-700}
    physicsBodyDefinitions [{
      sphereCollisionRadius=1
      boundBone {boneName=dyn_zipper_a_01}
      boxExtents=<8 12 12>
      constraintSetup {angularLimitsMax=<3 60 5>, angularLimitsMin=<-3 5 -20>, angularTarget=<0 0 0>}
      localJointOffset=<1 0 0>
     }]
    sphericalLimits [{limitRadius=5, limitType=Inner, drivingBone {boneName=base_zipper_a}, sphereLocalOffset=<2.2099 2.8148 -2.48>}]
   }
   n3 = Anim Dynamics - Bone: dyn_zipper_a_02!(ComponentPose=@n2, Alpha=1, GravityOverride:Vector=@n1, ExternalForce="0,0,0") -> (Pose) {
    angularDampingOverride=0.9
    bOverrideAngularDamping=true
    bUseGravityOverride=true
    simulationSpace=RootRelative
    alphaBoolBlend {blendOption=Linear}
    boundBone {boneName=dyn_zipper_a_02}
    componentAppliedLinearAccClamp=<600 600 600>
    componentLinearAccScale=<0.45 0.45 0.45>
    componentLinearVelScale=<0.6 0.6 0.6>
    gravityOverride {z=-700}
    physicsBodyDefinitions [{
      collisionType=CustomSphere
      sphereCollisionRadius=1.5
      boundBone {boneName=dyn_zipper_a_02}
      boxExtents=<6 8 9>
      constraintSetup {angularLimitsMax=<5 10 20>, angularLimitsMin=<-5 -20 -20>, angularTarget=<0 0 0>}
      localJointOffset=<0.5 0 0>
     }]
    planarLimits [{drivingBone {boneName=base_zipper_a}, planeTransform {location=<-0.6171 -0.7065 0.8367>, rotation=<rot -1.629 -6.6119 143.1549>}}]
    relativeSpaceBone {boneName=dyn_zipper_a_01}
    sphericalLimits [{limitRadius=5, limitType=Inner, drivingBone {boneName=base_zipper_a}, sphereLocalOffset=<2.9779 2.3027 -2.1615>}]
   }
   n4 = Anim Dynamics - Bone: dyn_zipper_b_01!(ComponentPose=@n3, Alpha=1, GravityOverride:Vector=@n1, ExternalForce="0,0,0") -> (Pose) {
    angularDampingOverride=0.9
    bOverrideAngularDamping=true
    bUseGravityOverride=true
    bUsePlanarLimit=false
    simulationSpace=RootRelative
    alphaBoolBlend {blendOption=Linear}
    boundBone {boneName=dyn_zipper_b_01}
    componentAppliedLinearAccClamp=<600 600 600>
    componentLinearAccScale=<0.45 0.45 0.45>
    componentLinearVelScale=<0.6 0.6 0.6>
    gravityOverride {z=-700}
    physicsBodyDefinitions [{
      sphereCollisionRadius=1
      boundBone {boneName=dyn_zipper_b_01}
      boxExtents=<8 12 12>
      constraintSetup {angularLimitsMax=<5 10 -5>, angularLimitsMin=<-5 -20 -60>, angularTarget=<0 0 0>}
      localJointOffset=<1 0 0>
     }]
    sphericalLimits [{limitRadius=5, limitType=Inner, drivingBone {boneName=base_zipper_b}, sphereLocalOffset=<2.1515 -1.7193 3.1589>}]
   }
   n5 = Anim Dynamics - Bone: dyn_zipper_b_02!(ComponentPose=@n4, Alpha=1, GravityOverride:Vector=@n1, ExternalForce="0,0,0") -> (Pose) {
    angularDampingOverride=0.9
    bOverrideAngularDamping=true
    bUseGravityOverride=true
    simulationSpace=RootRelative
    alphaBoolBlend {blendOption=Linear}
    boundBone {boneName=dyn_zipper_b_02}
    componentAppliedLinearAccClamp=<600 600 600>
    componentLinearAccScale=<0.45 0.45 0.45>
    componentLinearVelScale=<0.6 0.6 0.6>
    gravityOverride {z=-700}
    physicsBodyDefinitions [{
      collisionType=CustomSphere
      sphereCollisionRadius=1.5
      boundBone {boneName=dyn_zipper_b_02}
      boxExtents=<6 8 8>
      constraintSetup {angularLimitsMax=<5 10 20>, angularLimitsMin=<-5 -20 -30>, angularTarget=<0 0 0>}
      localJointOffset=<0.5 0 0>
     }]
    planarLimits [{drivingBone {boneName=base_zipper_b}, planeTransform {location=<-0.3942 0.4736 -0.6214>, rotation=<rot -16.456 10.9841 -52.3658>}}]
    sphericalLimits [{limitRadius=5, limitType=Inner, drivingBone {boneName=base_zipper_b}, sphereLocalOffset=<4.1305 -1.6644 2.8651>}]
   }
   n6 = "Transform (Modify) Bone - Bone: dyn_clavicle_l"!(
    ComponentPose=@n5
    bAlphaBoolEnabled:bool=True
    Translation=Get Left Clavicle Translations() -> (Left Clavicle Translations:Vector)
    Rotation=Get Left Clavicle Rotations() -> (Left Clavicle Rotations:Rotator)
    Scale="1,1,1"
   ) -> (Pose) {
    alphaInputType=Bool
    rotationMode=BMM_Replace
    rotationSpace=BCS_WorldSpace
    translationMode=BMM_Replace
    translationSpace=BCS_WorldSpace
    alphaBoolBlend {blendInTime=0.4, blendOption=Linear, blendOutTime=0.3}
    boneToModify {boneName=dyn_clavicle_l}
   }
   n7 = "Transform (Modify) Bone - Bone: dyn_clavicle_r"!(
    ComponentPose=@n6
    bAlphaBoolEnabled:bool=True
    Translation=Get Right Clavicle Translations() -> (Right Clavicle Translations:Vector)
    Rotation=Get Right Clavicle Rotations() -> (Right Clavicle Rotations:Rotator)
    Scale="1,1,1"
   ) -> (Pose) {
    alphaInputType=Bool
    rotationMode=BMM_Replace
    rotationSpace=BCS_WorldSpace
    translationMode=BMM_Replace
    translationSpace=BCS_WorldSpace
    alphaBoolBlend {blendInTime=0.4, blendOption=Linear, blendOutTime=0.3}
    boneToModify {boneName=dyn_clavicle_r}
   }
   n8 = "Transform (Modify) Bone - Bone: pack"!(ComponentPose=@n7, bAlphaBoolEnabled:bool=Get bIsSkydiveFloating(), Translation="0,-3.5,-0.5", Rotation="0,4,-7", Scale) -> (Pose) {
    alphaInputType=Bool
    rotationMode=BMM_Additive
    rotationSpace=BCS_BoneSpace
    translationMode=BMM_Additive
    translationSpace=BCS_BoneSpace
    alphaBoolBlend {blendInTime=0.3, blendOption=Linear, blendOutTime=2.5}
    boneToModify {boneName=pack}
    rotation {roll=-7, yaw=4}
    translation {y=-3.5, z=-0.5}
   }
   n9 = Control Rig!(Source=Component To Local(ComponentPose=@n8), bAlphaBoolEnabled:bool=True) -> (Pose) {
    alphaInputType=Bool
    alphaBoolBlend {blendOption=Linear}
    controlRigAssetReference {blueprintRigClass=&/M_MED_SherWolf_Pack_ControlRig.M_MED_SherWolf_Pack_ControlRig_C}
   }
   Output Pose(Result=@n9) {layerGroup=None}
  ]}
 EventGraph {nodes [
   Event BlueprintUpdateAnimation()
   > n1 = Gravity Override State Function DEPRECATED(
    CharacterPartAnimInstance=Self-Reference()
    GameTimeInSeconds=GetGameTimeInSeconds()
    GravityOverrideParams=Get GravityOverrideParams() -> (GravityOverrideParams:Gravity Override Params Struct Structure)
   )
   > Set Gravity Override(Gravity Override=@n1)
   > n2 = Cast To PlayerPawn_Athena(Object=TryGetPawnOwner())
   > n3 = Set Body Mesh(Body Mesh:Skeletal Mesh Component Object Reference=GetSkeletalMeshForPartType(self=@n2)) -> (Output_Get:Skeletal Mesh Component Object Reference)
   n4 = BreakTransform(InTransform=GetSocketTransform(self=@n3, InSocketName=clavicle_l, TransformSpace=RTS_World))
   > Set Left Clavicle Translations(Left Clavicle Translations:Vector=@n4) -> (Output_Get:Vector)
   > Set Left Clavicle Rotations(Left Clavicle Rotations:Rotator=@n4.Rotation) -> (Output_Get:Rotator)
   n5 = BreakTransform(InTransform=GetSocketTransform(self=@n3, InSocketName=clavicle_r, TransformSpace=RTS_World))
   > Set Right Clavicle Translations(Right Clavicle Translations:Vector=@n5) -> (Output_Get:Vector)
   > Set Right Clavicle Rotations(Right Clavicle Rotations:Rotator=@n5.Rotation) -> (Output_Get:Rotator)
   TryGetPawnOwner()
  ]}
]
variables [
 Body Mesh {type=SkeletalMeshComponent}
 Gravity Override {type=Vector}
 GravityOverrideParams {type=GravityOverrideParamsStruct}
 Left Clavicle Rotations {type=Rotator}
 Left Clavicle Translations {type=Vector}
 Right Clavicle Rotations {type=Rotator}
 Right Clavicle Translations {type=Vector}
]
