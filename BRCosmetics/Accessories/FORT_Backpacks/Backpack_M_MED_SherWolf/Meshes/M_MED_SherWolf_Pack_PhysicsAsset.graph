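# Physics asset for the SherWolf backpack cosmetic mesh, giving its two zipper pulls physical bodies for secondary motion.
# Each zipper (a and b) is a chain of small capsules: a base body followed by two dynamic segments.
# Constraints link each segment to the one before it, with tight swing limits and almost no twist, so each pull sways only slightly from its base.
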PhysicsAsset /BRCosmetics/Accessories/FORT_Backpacks/Backpack_M_MED_SherWolf/Meshes/M_MED_SherWolf_Pack_PhysicsAsset
bodies {
 base_zipper_a [{shapeType=Capsule, length=0.505, radius=0.505, shapeName=base_zipper_a_capsule, rotation=<rot 0 -90 0>}]
 base_zipper_b [{shapeType=Capsule, length=0.505, radius=0.505, shapeName=base_zipper_b_capsule, rotation=<rot 0 -90 0>}]
 dyn_zipper_a_01 [{shapeType=Capsule, length=1, radius=0.5, shapeName=dyn_zipper_a_01_capsule, center=<2 0 0>, rotation=<rot 90 0 0>}]
 dyn_zipper_a_02 [{shapeType=Capsule, length=1, radius=0.5, shapeName=dyn_zipper_a_02_capsule, center=<2 0 0>, rotation=<rot 90 0 0>}]
 dyn_zipper_b_01 [{shapeType=Capsule, length=1, radius=0.5, shapeName=dyn_zipper_b_01_capsule, center=<2 0 0>, rotation=<rot 90 0 0>}]
 dyn_zipper_b_02 [{shapeType=Capsule, length=1, radius=0.5, shapeName=dyn_zipper_b_02_capsule, center=<2 0 0>, rotation=<rot 90 0 0>}]
}
constraints [
 {bone1Name=dyn_zipper_a_01, bone2Name=base_zipper_a, swing1LimitDegrees=15, swing2LimitDegrees=15, twistLimitDegrees=1}
 {bone1Name=dyn_zipper_a_02, bone2Name=dyn_zipper_a_01, swing1LimitDegrees=15, swing2LimitDegrees=15, twistLimitDegrees=1}
 {bone1Name=dyn_zipper_b_01, bone2Name=base_zipper_b, swing1LimitDegrees=15, swing2LimitDegrees=15, twistLimitDegrees=1}
 {bone1Name=dyn_zipper_b_02, bone2Name=dyn_zipper_b_01, swing1LimitDegrees=15, swing2LimitDegrees=15, twistLimitDegrees=1}
]
